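# Conversation database for the Wild Card NPC in the NPC Library, defining the greeting players hear when they talk to this character.
# A single entry point leads to a speech node with a default line plus contextual lines for a first-ever conversation or specific worn outfits.
# Curve-driven priorities rank the first-meeting and outfit reactions.
# The speech node then branches through link nodes to the shared NPC Library quests and basic services conversations.
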ConversationDatabase /NPCLibrary/NPCs/WildCard/NPCLibrary_Conversation_WildCard
entry_points [{tag=Conversation.Entry.WildCard, destinations [#1]}]
nodes [
 ConversationEntryPointNode_7 {
  id=#1
  class=ConversationEntryPointNode
  outputs [#2]
  properties {entryTag {tagName=Conversation.Entry.WildCard}}
 }
 Speech_C_4 {
  id=#2
  class=Speech_C
  outputs [#3, #4]
  properties {
   nodeName=""
   generalConfig {
    defaultMessage=Ante up or fold. Your choice.
    contextualMessages [
     {
      message="If you're too afraid to go all-in, you shouldn't even play."
      contextRequirements [{requirement {class=FortControllerRequirement_IsFirstEverConversationWithNPC}}]
      priority {curve {rowName=Tandem.Speech.Priority.FirstEverInteraction}}
     }
     {
      message=Tell me more about this midnight grind...
      contextRequirements [{requirement {class=FortControllerRequirement_HasCID, properties {allowedCIds [/BRCosmetics/Athena/Items/Cosmetics/Characters/Character_Venice]}}}]
      priority {curve {rowName=Tandem.Speech.Priority.OutfitReaction}}
     }
     {
      message=You're a little goodie-two-shoes for my liking.
      contextRequirements [{requirement {class=FortControllerRequirement_HasCID, properties {allowedCIds [&Character_Sunlit]}}}]
      priority {curve {rowName=Tandem.Speech.Priority.OutfitReaction}}
     }
     {
      message=You seem willing to get your hands dirty. Ever done a heist?
      contextRequirements [{requirement {class=FortControllerRequirement_HasCID, properties {allowedCIds [/BRCosmetics/Athena/Items/Cosmetics/Characters/Character_Chainmail]}}}]
      priority {curve {rowName=Tandem.Speech.Priority.OutfitReaction}}
     }
     {
      message="I swear, your majesty. I'm a model citizen."
      contextRequirements [{requirement {class=FortControllerRequirement_HasCID, properties {allowedCIds [&Character_Citadel]}}}]
      priority {curve {rowName=Tandem.Speech.Priority.OutfitReaction}}
     }
     {
      message=The heist is off. Go find a ruin to plunder.
      contextRequirements [{requirement {
         class=FortControllerRequirement_HasCID
         properties {allowedCIds [/BRCosmetics/Athena/Items/Cosmetics/Characters/CID_397_Athena_Commando_F_TreasureHunterFashion]}
        }}]
      priority {curve {rowName=Tandem.Speech.Priority.OutfitReaction}}
     }
     {
      message=I got a better offer from our mark. No hard feelings.
      contextRequirements [{requirement {
         class=FortControllerRequirement_HasCID
         properties {allowedCIds [/BRCosmetics/Athena/Items/Cosmetics/Characters/CID_936_Athena_Commando_F_RaiderSilver]}
        }}]
      priority {curve {rowName=Tandem.Speech.Priority.OutfitReaction}}
     }
     {
      message=The Syndicate's not getting a cut of my take.
      contextRequirements [{requirement {
         class=FortControllerRequirement_HasCID
         properties {allowedCIds [&CID_A_425_Athena_Commando_F_BlueJay]}
        }}]
      priority {curve {rowName=Tandem.Speech.Priority.OutfitReaction}}
     }
     {
      message="I wish you'd taken me up on my offer, love."
      contextRequirements [{requirement {
         class=FortControllerRequirement_HasCID
         properties {allowedCIds [/BRCosmetics/Athena/Items/Cosmetics/Characters/CID_A_478_Athena_Commando_F_WildCard]}
        }}]
      priority {curve {rowName=Tandem.Speech.Priority.OutfitReaction}}
     }
     {
      message=You should've taken me up on my offer.
      requirementMatchPolicy=RequireAny
      contextRequirements [{requirement {
         class=FortControllerRequirement_HasCID
         properties {allowedCIds [
           /BRCosmetics/Athena/Items/Cosmetics/Characters/CID_A_428_Athena_Commando_F_PinkWidow
           /BRCosmetics/Athena/Items/Cosmetics/Characters/CID_462_Athena_Commando_M_HeistSummer
           /BRCosmetics/Athena/Items/Cosmetics/Characters/CID_A_129_Athena_Commando_M_CatBurglarSummer
           /BRCosmetics/Athena/Items/Cosmetics/Characters/CID_A_257_Athena_Commando_M_CatBurglar_Ghost
           /BRCosmetics/Athena/Items/Cosmetics/Characters/CID_A_320_Athena_Commando_M_CatburglarWinter
           /BRCosmetics/Athena/Items/Cosmetics/Characters/CID_A_390_Athena_Commando_M_Blackbird
           /BRCosmetics/Athena/Items/Cosmetics/Characters/CID_694_Athena_Commando_M_CatBurglar
           /BRCosmetics/Athena/Items/Cosmetics/Characters/CID_923_Athena_Commando_M_ParcelGold
          ]}
        }}]
      priority {curve {rowName=Tandem.Speech.Priority.OutfitReaction}}
     }
    ]
   }
  }
 }
 ConversationLinkNode_0 {id=#4, class=ConversationLinkNode, properties {remoteEntryTag {tagName=Conversation.Entry.BasicServices}}}
 ConversationLinkNode_1 {id=#3, class=ConversationLinkNode, properties {remoteEntryTag {tagName=Conversation.Entry.NPCLibrary.Quests}}}
]
